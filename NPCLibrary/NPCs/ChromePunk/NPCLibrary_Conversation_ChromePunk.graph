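# Conversation database for the Chrome Punk NPC, defining its greeting dialogue and the menus that follow it.
# The entry point leads to one speech node with a default line, swapped for contextual lines on a first-ever conversation or when the player wears specific cosmetic characters.
# That speech then branches through link nodes into the shared NPC library quests and basic services conversations.
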ConversationDatabase /NPCLibrary/NPCs/ChromePunk/NPCLibrary_Conversation_ChromePunk
entry_points [{tag=Conversation.Entry.ChromePunk, destinations [#1]}]
nodes [
 ConversationEntryPointNode_7 {
  id=#1
  class=ConversationEntryPointNode
  outputs [#2]
  properties {entryTag {tagName=Conversation.Entry.ChromePunk}}
 }
 Speech_C_4 {
  id=#2
  class=Speech_C
  outputs [#3, #4]
  properties {
   nodeName=""
   generalConfig {
    defaultMessage="What do you want from me, huh?"
    contextualMessages [
     {
      message="Hey there, pal. You stuck here too?"
      contextRequirements [{requirement {class=FortControllerRequirement_IsFirstEverConversationWithNPC}}]
      priority {curve {rowName=Tandem.Speech.Priority.FirstEverInteraction}}
     }
     {
      message="Didn't think I'd ever see you again, bro."
      contextRequirements [{requirement {
         class=FortControllerRequirement_HasCID
         properties {allowedCIds [/BRCosmetics/Athena/Items/Cosmetics/Characters/CID_902_Athena_Commando_M_PumpkinPunk]}
        }}]
      priority {value=0, curve {curveTable=null, rowName=None}}
     }
     {
      message="Yo, I heard you can spit bars."
      contextRequirements [{requirement {
         class=FortControllerRequirement_HasCID
         properties {allowedCIds [/BRCosmetics/Athena/Items/Cosmetics/Characters/CID_344_Athena_Commando_M_IceCream]}
        }}]
      priority {value=0, curve {curveTable=null, rowName=None}}
     }
     {
      message=Are we...cousins or something?
      contextRequirements [{requirement {
         class=FortControllerRequirement_HasCID
         properties {allowedCIds [/BRCosmetics/Athena/Items/Cosmetics/Characters/CID_244_Athena_Commando_M_PumpkinSuit]}
        }}]
      priority {value=0, curve {curveTable=null, rowName=None}}
     }
     {
      message=Your hands match my face.
      contextRequirements [{requirement {class=FortControllerRequirement_HasCID, properties {allowedCIds [&Character_Bites]}}}]
      priority {value=0, curve {curveTable=null, rowName=None}}
     }
     {
      message=I suppose I have you to thank for the new drip.
      contextRequirements [{requirement {class=FortControllerRequirement_HasCID, properties {allowedCIds [&Character_TheHerald]}}}]
      priority {value=0, curve {curveTable=null, rowName=None}}
     }
     {
      message=I don't normally like cats but you seem pretty cool.
      contextRequirements [{requirement {class=FortControllerRequirement_HasCID, properties {allowedCIds [&Character_ChillCat]}}}]
      priority {value=0, curve {curveTable=null, rowName=None}}
     }
     {
      message=One's a bear. One has a pumpkin head. They fight crime.
      contextRequirements [{requirement {class=FortControllerRequirement_HasCID, properties {allowedCIds [&Character_BadBear]}}}]
      priority {value=0, curve {curveTable=null, rowName=None}}
     }
     {
      message="Sorry, never saw your show. I heard it's good though."
      contextRequirements [{requirement {class=FortControllerRequirement_HasCID, properties {allowedCIds [&Character_PinkSpike]}}}]
      priority {value=0, curve {curveTable=null, rowName=None}}
     }
     {
      message=I'm not the guy you're thinking of but I understand the confusion.
      contextRequirements [{requirement {class=FortControllerRequirement_HasCID, properties {allowedCIds [&Character_RoseDust]}}}]
      priority {value=0, curve {curveTable=null, rowName=None}}
     }
    ]
   }
  }
 }
 ConversationLinkNode_0 {id=#4, class=ConversationLinkNode, properties {remoteEntryTag {tagName=Conversation.Entry.BasicServices}}}
 ConversationLinkNode_1 {id=#3, class=ConversationLinkNode, properties {remoteEntryTag {tagName=Conversation.Entry.NPCLibrary.Quests}}}
]
